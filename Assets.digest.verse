# Copyright Epic Games, Inc. All Rights Reserved.
#################################################
# Generated Digest of Verse API
# DO NOT modify this manually!
# Generated from build: ++Fortnite+Release-37.00-CL-44501951
#################################################


using {/Verse.org/Native}
using {/Verse.org/SceneGraph}
# Module import path: /playkigai.dig@fortnite.com/Minigame_TugOfWar/EntityPrefabs
EntityPrefabs := module:
    @import_as("/Minigame_TugOfWar/EntityPrefabs/tow_game_template.tow_game_template_C")
    tow_game_template<scoped {/playkigai.dig@fortnite.com/Minigame_TugOfWar}> := class<final>(entity):


    tow_game_template_asset<scoped {/playkigai.dig@fortnite.com/Minigame_TugOfWar}>:entity_prefab = external {}

    @import_as("/Minigame_TugOfWar/EntityPrefabs/tow_player_point.tow_player_point_C")
    tow_player_point<scoped {/playkigai.dig@fortnite.com/Minigame_TugOfWar}> := class<final>(entity):


    tow_player_point_asset<scoped {/playkigai.dig@fortnite.com/Minigame_TugOfWar}>:entity_prefab = external {}

    @import_as("/Minigame_TugOfWar/EntityPrefabs/tow_rope.tow_rope_C")
    tow_rope<scoped {/playkigai.dig@fortnite.com/Minigame_TugOfWar}> := class<final>(entity):


    tow_rope_asset<scoped {/playkigai.dig@fortnite.com/Minigame_TugOfWar}>:entity_prefab = external {}