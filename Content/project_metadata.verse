<#
@UUID: 69d6d5ab-9db0-47f1-8c7c-d3efb805664d
@Task: Tug of War Minigame
@Description: Create a Tug of War game in verse using scene graph. Before the game begins, players should be able to 
    register to a side by interacting with premade points on the rope. Once registered, their character should be placed 
    in stasis, their camera view should be changed to a fixed camera device overlooking the entire rope, and their character's
    position should be snapped to be beside their point they registered to. Pressing the interact button again before the game starts,
    Once there are enough players registered on both sides the game should start. Players pressing the interact button during the game
    should cause the rope to be pulled in their direction and pulling the opposite team toward them. If any player falls into the pit in
    the center of the rope, they should be eliminated and disqualified. The last team standing is the winner.
@Tags: Noncombat, PvP, Action, Sports
@Export: true
#>
