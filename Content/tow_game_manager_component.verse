<#
@UUID: 69d6d5ab-9db0-47f1-8c7c-d3efb805664d
@Task: Tug of War Game Manager Component
@Description: Handles the High level logic of the Tug of War Minigame. Allows players to register to the game by interacting to point on the rope, manages imputs for both teams, and determines a winner when all players on a given team are eliminated
@Tags: SceneGraph, Gameplay, TeamManagement, PlayerManagement, Movement, Devices
@Export: true
#>

<#
In-Editor Entity Prefab Setup:
    (NOTE: any setting not mentioned is assumed to be default or to your preference)


tow_player_point
    RootEntity:
        - tow_player_point_component
        InteractableCollider
            - interactable_component
            - mesh_component


tow_rope
    RootEntity:
        RopeMesh
            - mesh_component (preferably cylinder)
            - transform_component
                - rotate rope so that it is horizontal and adjust the scale so that it stretches out like a long rope
        Side1PlayerPoints
            - transform_component
                - move to the left side of the rope mesh
            1-A (tow_player_point prefab instance)
            1-B (tow_player_point prefab instance)
            1-C (tow_player_point prefab instance)
        Side2PlayerPoints
            - transform_component
                    - move to the right side of the rope mesh
            2-A (tow_player_point prefab instance)
            2-B (tow_player_point prefab instance)
            2-C (tow_player_point prefab instance)


tow_game_template
    RootEntity:
        - tow_game_manager_component
            - MaybeInputDevice: set to an volume_device device set to the interact button in the editor
            - MaybeEliminationVolume: set to a volume_device placed in the center of the rope entity.
            - MaybeCamDevice: set to a gameplay_camera_fixed_point_device angled such that it is horizontally looking at the rope.
            - MaybeRopePrefab: set to RopeEntity
            - Side1PlayerPoints: populate with all the tow_player_point prefab entities under Side1PlayerPoints under the RopeEntity
            - Side2PlayerPoints: populate with all the tow_player_point prefab entities under Side2PlayerPoints under the RopeEntity
        - tow_hud_component
    RopeEntity (tow_rope prrefab instance)
        - transform_component
            - position slightly above the above the ground at about half the height of a fortnite character

#>

using { /UnrealEngine.com/Temporary/UI }
using { /UnrealEngine.com/Temporary/SpatialMath }
using { /Fortnite.com/UI }
using { /Fortnite.com/Devices }
using { /Fortnite.com/Teams }
using { /Fortnite.com/Playspaces }
using { /Verse.org }    
using { /Verse.org/Colors }
using { /Verse.org/Native }
using { /Verse.org/SceneGraph }
using { /Verse.org/Simulation }
using { /Fortnite.com/Game }
using { /Fortnite.com/Characters }
using { /Verse.org/SpatialMath }
using {EntityPrefabs}


tow_ready_status_update_result := struct:
    IsSide1Ready: logic
    IsSide2Ready: logic

game_state := enum {PreGame, InProgress, PostGame}

tow_game_manager_component<public> := class<final_super>(component):

    # ============ EDITABLE FIELDS ==================================
    # The amount of time in seconds after the game that this component will wait before reseting in anticipation for another
    @editable_number(float):
        MinValue := option. 0.0
    PostGameDelay: float = 5.0
    
    # input_trigger_device that signals when any player registered to this game presses the interact button
    @editable 
    MaybeInputDevice: ?input_trigger_device = false
    
    # volume_device that sits in pit in the the middle of the rope, used to tell the game to remove the falling player from stasis so they can be eliminated
    @editable 
    MaybeEliminationVolume: ?volume_device = false
    
    # The gameplay_camera_fixed_point_device that overlooks the entire rope
    @editable
    MaybeCamDevice: ?gameplay_camera_fixed_point_device = false
    
    # The child tow_rope prefab root entity to be moved. Set in the editor
    @editable 
    MaybeRopePrefab: ?tow_rope = false

    # The global distance the rope is moved each time a player presses the interact button during the game
    @editable 
    DistancePerPress: float = 10.0
    
    # The arrays that contain the player points on each side of the rop; set in the editor
    @editable
    Side1PlayerPoints: []tow_player_point = array{}
    @editable
    Side2PlayerPoints: []tow_player_point = array{}

    # ============ EVENTS =======================================
    # Signaled when the game starts
    GameBeginEvent<public>: event() = event(){}
    # Signals at the end of the game (regardless of the result); passes a logic represetning which side is the winner: Side1 if True, Side 2 otherwise
    GameOverEvent: event(logic) = event(logic){}
    # Signals when the game is meant to reset
    GameResetEvent<public>: event() = event(){}
    # Signals when the game manager fully reset in anticipation for the next game
    PostGameDelayCompleteEvent<public>: event() = event(){}
    # Signals when the ready status changes for a player point
    ReadyStatusUpdateEvent: event(tow_ready_status_update_result) = event(tow_ready_status_update_result){}
    # Signals when this component's OnEndSimulation function fires; used to cancel asyncronous logic when this component stops simulating
    EndSimulationEvent<internal>: event() = event(){}
    # Fires when a player has joined, left the game, or died
    SidePlayersUpdatedEvent<internal>: event() = event(){}
    
    # ============ INTERNAL VARS =======================================
    var Side1PointComps: []tow_player_point_component = array{}
    var Side2PointComps: []tow_player_point_component = array{}
    var AgentToPoint: [agent]tow_player_point_component = map{}
    var MaybeHUDComponent: ?tow_hud_component = false 
    var GameState: game_state = game_state.PreGame
    var RopeTransformOrigin: (/Verse.org/SpatialMath:) transform = (/Verse.org/SpatialMath:) transform{}


    OnBeginSimulation<override>():void=
        (super:)OnBeginSimulation()
        for (PlayerPointPF: Side1PlayerPoints, PlayerPointComp := PlayerPointPF.GetPointComp[]):
           set Side1PointComps += array{PlayerPointComp}
        for (PlayerPointPF: Side2PlayerPoints, PlayerPointComp := PlayerPointPF.GetPointComp[]):
           set Side2PointComps += array{PlayerPointComp}
           set PlayerPointComp.IsSide1 = false
        if (RoundMan := Entity.GetFortRoundManager[]):
            RoundMan.SubscribeRoundStarted(OnRoundStart)


    OnRoundStart()<suspends>:void=
        if(VolDevice := MaybeEliminationVolume?):
            VolDevice.AgentEntersEvent.Subscribe(OnAgentEnterVolume)

        if(RopeEntity := MaybeRopePrefab?):
            set RopeTransformOrigin = RopeEntity.GetGlobalTransform()

        set MaybeHUDComponent = option. Entity.GetComponent[tow_hud_component]
        race:
            EndSimulationEvent.Await()
            loop:
                # On Pre Game
                set GameState = game_state.PreGame
                ResetGame()
                spawn. ListenForPlayerInput()
                SpawnPlayerRegisterListeners()
                WaitForBothSidesReady()
                # On Game Begin
                GameBeginEvent.Signal()
                set GameState = game_state.InProgress
                # On Post Game
                if(HUDComp := MaybeHUDComponent?):
                    Winner := ListenForIsSideOneLeft()
                    GameOverEvent.Signal(Winner)
                set GameState = game_state.PostGame
                UnregisterAllTowPointComponents()
                # Once the game is over, displaying results is handled by the HUD Component
                Sleep(PostGameDelay)
                PostGameDelayCompleteEvent.Signal()


    OnEndSimulation<override>():void=
        (super:)OnEndSimulation()
        EndSimulationEvent.Signal()
        ResetGame()


    OnAgentEnterVolume(PlayerAgent : agent):void=
        if:
            FortChar := PlayerAgent.GetFortCharacter[]
            PointComp := AgentToPoint[PlayerAgent]
        then:
            PointComp.UnregisterAgent()


    UnregisterAllTowPointComponents():void=
        for (PointComp : Side1PointComps):
            PointComp.UnregisterAgent()
        for (PointComp : Side2PointComps):
            PointComp.UnregisterAgent()


    WaitForBothSidesReady()<suspends>:void=
        race:
            EndSimulationEvent.Await()
            GameResetEvent.Await()
            loop:
                SidePlayersUpdatedEvent.Await()
                isSide1Ready :=  IsSideReady(Side1PointComps)
                isSide2Ready := IsSideReady(Side2PointComps)
                ReadyStatusUpdateEvent.Signal(tow_ready_status_update_result{IsSide1Ready := isSide1Ready, IsSide2Ready := isSide2Ready})
                if(isSide1Ready? and isSide2Ready?):
                    break
                Sleep(0.0)


    SpawnPlayerRegisterListeners()<suspends>:void=
        for (PointComp : Side1PointComps):
            spawn. ListenForRegister(PointComp, true)
        for (PointComp : Side2PointComps):
            spawn. ListenForRegister(PointComp, false)


    ListenForIsSideOneLeft()<suspends>:logic=
        race:
            EndSimulationEvent.Await()
            GameResetEvent.Await()
            loop:
                SidePlayersUpdatedEvent.Await()
                if(CheckForSideEmpty[Side1PointComps]):
                    return false
                if(CheckForSideEmpty[Side2PointComps]):
                    return true
                Sleep(0.0)
        return true


    CheckForSideEmpty(SidePointComps : []tow_player_point_component)<transacts><decides>:void=
        for:
            PlayerPointComp : SidePointComps
            PlayerPointComp.MaybeRegisteredAgent?
        do:
            false?
         

    IsSideReady(PointCompArray : []tow_player_point_component):logic=
        var isReady : logic = false
        for:
            PlayerPointComp : PointCompArray
            PlayerPointComp.MaybeRegisteredAgent?
        do:
            set isReady = true
        isReady


    ListenForRegister(PointComp : tow_player_point_component, isSide1 : logic)<suspends>:void=
        race:
            EndSimulationEvent.Await()
            GameResetEvent.Await()
            GameBeginEvent.Await()
            loop:
                RegisteredAgent := PointComp.NewAgentRegisteredEvent.Await()

                if (set AgentToPoint[RegisteredAgent] = PointComp):
                    if (CamDevice := MaybeCamDevice?):
                        CamDevice.AddTo(RegisteredAgent)
                    if (InputDevice := MaybeInputDevice?):
                        InputDevice.Register(RegisteredAgent)
                    
                    spawn. ListenForUnregister(PointComp)
                    SidePlayersUpdatedEvent.Signal()
                Sleep(0.0)


    ListenForUnregister(PointComp : tow_player_point_component)<suspends>:void=
        race:
            EndSimulationEvent.Await()
            GameResetEvent.Await()
            block:
                UnregisteredAgent := PointComp.AgentUnregisteredEvent.Await()
                set AgentToPoint = AgentToPoint.RemoveKey(UnregisteredAgent)
                if (CamDevice := MaybeCamDevice?):
                    CamDevice.RemoveFrom(UnregisteredAgent)
                if (InputDevice := MaybeInputDevice?):
                    InputDevice.Unregister(UnregisteredAgent)
                SidePlayersUpdatedEvent.Signal()


    ListenForPlayerInput()<suspends>:void=
        if (InputDevice := MaybeInputDevice?):
            race:
                EndSimulationEvent.Await()
                GameResetEvent.Await()
                loop:
                    PlayerAgent := InputDevice.PressedEvent.Await()
                    if (PointComp := AgentToPoint[PlayerAgent]):
                        case (GameState):
                            game_state.PreGame =>
                                PointComp.UnregisterAgent()
                            game_state.InProgress =>
                                MoveRope(PointComp.IsSide1)
                            game_state.PostGame =>
                                PointComp.UnregisterAgent()
                    Sleep(0.0)
            InputDevice.UnregisterAll()


    MoveRope(isSide1 : logic)<suspends>:void=
        # Move entity according to side
        if(RopeEntity := MaybeRopePrefab?):
            var DisToMove : float = DistancePerPress
            if(not isSide1?):
                set DisToMove = DisToMove*-1.0
            RopeTransform := RopeEntity.GetGlobalTransform()
            NewRopeTransform := (/Verse.org/SpatialMath:) transform {
                Translation := (/Verse.org/SpatialMath:) vector3{Left := RopeTransform.Translation.Left, Forward := RopeTransform.Translation.Forward+DisToMove, Up := RopeTransform.Translation.Up}
                Rotation := RopeTransform.Rotation
                Scale := RopeTransform.Scale
            }
            RopeEntity.SetGlobalTransform(NewRopeTransform)


    ResetGame():void=
        GameResetEvent.Signal()
        set AgentToPoint = map{}
        for (PointComp : Side1PointComps):
            PointComp.Reset()
        for (PointComp : Side2PointComps):
            PointComp.Reset()
        if (CamDevice := MaybeCamDevice?) then CamDevice.RemoveFromAll()
        if (InputDevice := MaybeInputDevice?) then InputDevice.UnregisterAll()
        if (RopeEntity := MaybeRopePrefab?) then RopeEntity.SetGlobalTransform(RopeTransformOrigin)
        ReadyStatusUpdateEvent.Signal(tow_ready_status_update_result{IsSide1Ready := false, IsSide2Ready := false})