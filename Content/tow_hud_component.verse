<#
@UUID: 69d6d5ab-9db0-47f1-8c7c-d3efb805664d
@Task: Tow HUD Component
@Description: Creates a HUD element on each players' screen that Displays the name of the game, which side needs to ready up before the game starts, and which team won after the game is over
@Tags: SceneGraph, UI, DataStructures
@Export: true
#>

using { /UnrealEngine.com/Temporary/UI }
using { /UnrealEngine.com/Temporary/SpatialMath }
using { /Fortnite.com/UI }
using { /Fortnite.com/Devices }
using { /Fortnite.com/Teams }
using { /Fortnite.com/Playspaces }
using { /Verse.org }    
using { /Verse.org/Colors }
using { /Verse.org/Native }
using { /Verse.org/SceneGraph }
using { /Verse.org/Simulation }
using { /Fortnite.com/Game }


###################### Start UI Component ##########################################################################

tow_hud_component<public> := class<final_super>(component):

    var PlayerToUI:[player]tow_hud = map{} # Stores the UI object for each player
    var MaybeGameManager: ?tow_game_manager_component = false  # Reference to the sibling chess_game_manager_component
    
    OnBeginSimulation<override>():void =
        (super:)OnBeginSimulation()
        if (RoundMan := Entity.GetFortRoundManager[]):
            RoundMan.SubscribeRoundStarted(OnRoundStart)
        set MaybeGameManager = option. Entity.GetComponent[tow_game_manager_component] 
        if (Playspace := Entity.GetPlayspaceForEntity[]):
            Playspace.PlayerRemovedEvent().Subscribe(OnPlayerRemoved)
        

    OnRoundStart()<suspends>:void=
        Sleep(0.0)
        if (GameManager := MaybeGameManager?):
            AddUIAllPlayers()
            loop:
                # Waiting For Player Ready
                race:
                    GameManager.GameBeginEvent.Await()
                    loop: # On Player Ready
                        UpdateReadyStatus(GameManager.ReadyStatusUpdateEvent.Await())
                        Sleep(0.0)
                for (Player->Hud : PlayerToUI):
                    Hud.ResetGameHUD()
                # Game Loop
                IsSide1Winner := GameManager.GameOverEvent.Await()
                # Post Game
                if(IsSide1Winner?):
                    UpdateResultText(SideWinMessage("Side 1"))
                else:
                    UpdateResultText(SideWinMessage("Side 2"))
                
                GameManager.GameResetEvent.Await()
                ResetUI()
                
                    
    OnEndSimulation<override>():void =
        RemoveUIAllPlayers()    # Remove all UI elements on end; Custom Verse UI does not get cleaned up automatically.
                

    UpdateResultText(Message:message):void = 
        for (Player->Hud : PlayerToUI):
            Hud.ResultText.SetText(Message)


    UpdateReadyStatus(UpdateResult : tow_ready_status_update_result): void = 
        for (Player -> Hud : PlayerToUI):
            var ReadyMessage : message = WaitingOnPlayers()

            if (not UpdateResult.IsSide1Ready? and not UpdateResult.IsSide2Ready?):
                Hud.ResultText.SetText(StringToMessage(""))
            else if (UpdateResult.IsSide2Ready?):
                Hud.ResultText.SetText(WaitingOnSide1())
            else if (UpdateResult.IsSide1Ready?):
                Hud.ResultText.SetText(WaitingOnSide2())

            
    ResetUI():void = 
        for (Player->Hud : PlayerToUI):
            Hud.ResetGameHUD()


    AddUIAllPlayers():void =
        if(Playspace := Entity.GetPlayspaceForEntity[]):
            for (Player : Playspace.GetPlayers()) do Player.AddUI()


    RemoveUIAllPlayers():void=
        if(Playspace := Entity.GetPlayspaceForEntity[]):
            for(Player : Playspace.GetPlayers()):
                Player.RemoveUI()
            set PlayerToUI = map{}
            

    (Player : player).AddUI(GameSideComponents : []tow_player_point_component):void =
        if:
            PlayerUI := GetPlayerUI[Player]
            ShuffleboardUI := tow_hud{}
            set PlayerToUI[Player] = ShuffleboardUI
            GameManager := MaybeGameManager?
        then:
            ShuffleboardUI.CreateUI()
            PlayerUI.AddWidget(ShuffleboardUI.GetWidget())
        else:
            Print("chess_hud_component:\nERROR: Could not add UI to player's screen")
        

    (Player : player).RemoveUI():void =
        if:
            PlayerUI := GetPlayerUI[Player]
            ShuffleboardUI := PlayerToUI[Player]
        then:
            PlayerUI.RemoveWidget(ShuffleboardUI.GetWidget())
            set PlayerToUI = PlayerToUI.RemoveKey(Player)
    

    OnPlayerAdded(Player: player):void = Player.AddUI()
    OnPlayerRemoved(Player: player):void = Player.RemoveUI()

###################### End UI Component #############################################################################


###################### Start UI Display Messages ######################################################################

TitleMessage<localizes>() : message = "Tug of War"
WaitingOnSide1<localizes>() : message = "Waiting for Side 1 to ready up"
WaitingOnSide2<localizes>() : message = "Waiting for Side 2 to ready up"
WaitingOnPlayers<localizes>() : message = "Waiting for both sides to be ready"
GameStartMessage<localizes>(String: string) : message = "Game Starting in {String}"
SideWinMessage<localizes>(String: string) : message = "{String} is the victor!"

###################### End UI Display Messages ##########################################################################


###################### Start Player UI Class ############################################################################
                
tow_hud := class:
    ResultText : text_block = text_block{DefaultTextColor := NamedColors.Yellow}
    var MainWidget : canvas = canvas{}
    GetWidget() : canvas = MainWidget

    ResetGameHUD() : void = 
        ResultText.SetText(StringToMessage(""))

    CreateUI() : void =
        GameTitle: text_block = text_block{DefaultTextColor := NamedColors.White, DefaultText := TitleMessage()},

        TitleUI := canvas:
            Slots := array:
                # Create title at the top of the screen
                canvas_slot:
                    Anchors := anchors{Minimum := vector2{X:= 0.5, Y := 0.0}, Maximum := vector2{X:= 0.5, Y := 0.0}}
                    Offsets := margin{Left := 0.0, Top := 130.0, Right := 650.0, Bottom := 100.0}
                    Alignment := vector2{X := 0.5, Y := 0.5}
                    Widget := overlay{
                        Slots := array{
                            overlay_slot:
                                Widget := color_block:
                                    DefaultColor := MakeColorFromSRGB(1.0, 0.07, 0.07)
                                    DefaultOpacity := 0.5
                                    DefaultDesiredSize := vector2{X := 400.0, Y := 80.0}  

                            overlay_slot{
                                Widget := GameTitle
                            }
                        }
                    }
                    

        OverallUI := stack_box{
            Orientation := orientation.Vertical,
            Slots := array{
                stack_box_slot{
                    Widget := TitleUI,
                    HorizontalAlignment := horizontal_alignment.Center,
                    VerticalAlignment := vertical_alignment.Center
                },
                stack_box_slot{
                    Widget := ResultText,
                    HorizontalAlignment := horizontal_alignment.Center,
                    VerticalAlignment := vertical_alignment.Center
                }
            }
        }

        set MainWidget = canvas{
            Slots := array{
                canvas_slot{
                    Widget := OverallUI,
                    Anchors := anchors{Minimum := vector2{X := 0.5, Y := 0.1}, Maximum := vector2{X := 0.5, Y := 0.1}},
                    Alignment := vector2{X := 0.5, Y := 0.5}
                }
            }
        }