<#
@UUID: 69d6d5ab-9db0-47f1-8c7c-d3efb805664d
@Task: Tug Of War Player Point Component
@Description: Entity that a player agent can register to by interacting with. While registered, the player's character model is placed into stasis and teleported to the point's location.
@Tags: SceneGraph, PlayerManagement, Interaction, Movement
@Export: true
#>

<#
In-Editor Entity Prefab Setup:
    (NOTE: any setting not mentioned is assumed to be default or to your preference)

tow_player_point
    RootEntity:
        - tow_player_point_component
        InteractableCollider
            - interactable_component
            - mesh_component
#>

using { /Verse.org }
using { /Verse.org/Native }
using { /Verse.org/SceneGraph }
using { /Verse.org/Simulation }
using { /Verse.org/SpatialMath }
using { /Fortnite.com/Game }
using { /Fortnite.com/Playspaces }
using { /Fortnite.com/Characters }
using { /Fortnite.com/Devices }
using { /UnrealEngine.com/Temporary/SpatialMath }
using {EntityPrefabs}


tow_player_point_component<public> := class<final_super>(component):
    @editable
    var SnapPointOffset: (/Verse.org/SpatialMath:) vector3 = (/Verse.org/SpatialMath:) vector3 {
        Forward := 0.0,
        Left := 60.0,
        Up := 0.0
    }

    var MaybeRegisteredAgent: ?agent = false
    var MaybeInteractable: ?interactable_component = false
    var MaybeAgentRemovedSub: ?cancelable = false
    var MaybeRegisteredAgentEliminatedSub: ?cancelable = false
    var IsSide1: logic = true

    # Signals when a new agent is registered to this point; passes the agent who was registered
    NewAgentRegisteredEvent: event(agent) = event(agent){}
    # Signals when a new agent is unregistered from this point; passes the previously registered agent
    AgentUnregisteredEvent: event(agent) = event(agent){}
    # Signals when this component's OnEndSimulation function is called; used to cancel asyncronous functions
    EndSimulationEvent: event() = event(){}


    OnBeginSimulation<override>():void =
        (super:)OnBeginSimulation()
        if (Interactable := Entity.FindDescendantComponentsArray(interactable_component)[0]):
            Interactable.SucceededEvent.Subscribe(RegisterAgent)
            set MaybeInteractable = option. Interactable


    OnEndSimulation<override>():void =
        (super:)OnEndSimulation()
        EndSimulationEvent.Signal()


    RegisterAgent(Instigator: agent):void = 
        if:
            Playspace := Entity.GetPlayspaceForEntity[]
            not MaybeRegisteredAgent?
            FortChar := Instigator.GetFortCharacter[]
        then:
            set MaybeRegisteredAgent = option. Instigator
            FortChar.PutInStasis(stasis_args {AllowTurning := false, AllowFalling := false, AllowEmotes := false})
            spawn. SnapAgentToPoint(Instigator)
            Sub := Playspace.ParticipantRemovedEvent().Subscribe(OnParticipantRemoved)
            set MaybeAgentRemovedSub = option. Sub
            NewAgentRegisteredEvent.Signal(Instigator)


    OnParticipantRemoved(Agent: agent): void=
        if (MaybeRegisteredAgent? = Agent):
            UnregisterAgent()
    

    UnregisterAgent():void = 
        if:
            RegisteredAgent := MaybeRegisteredAgent?
            FortChar := RegisteredAgent.GetFortCharacter[]
        then:
            set MaybeRegisteredAgent = false
            FortChar.ReleaseFromStasis()
            if. MaybeAgentRemovedSub?.Cancel()
            AgentUnregisteredEvent.Signal(RegisteredAgent)
            if. MaybeRegisteredAgentEliminatedSub?.Cancel()


    SnapAgentToPoint<internal>(Agent: agent)<suspends>:void =
        if (FortChar := Agent.GetFortCharacter[]):
            race:
                EndSimulationEvent.Await()
                AgentUnregisteredEvent.Await()
                loop:
                    Sleep(0.0)
                    EntityTransform := FromTransform(Entity.GetGlobalTransform()) 
                    if. FortChar.TeleportTo[EntityTransform.Translation + FromVector3(SnapPointOffset),EntityTransform.Rotation]


    Reset():void=
        UnregisterAgent()
        if. MaybeAgentRemovedSub?.Cancel()
        if. MaybeRegisteredAgentEliminatedSub?.Cancel()


(PlayerPointPF: tow_player_point).GetPointComp()<decides><transacts>:tow_player_point_component = 
    PlayerPointPF.GetComponent[tow_player_point_component]


    
