<#
@UUID: 69d6d5ab-9db0-47f1-8c7c-d3efb805664d
@Task: Tug Of War Utility Functions
@Description: Helper functions used across the Tug Of War Project to traverse scene graph trees, convert data, and manage data structures
@Tags: SceneGraph, Utility, DataConversion, DataStructures
@Export: true
#>

using { /Verse.org }
using { /Verse.org/Native }
using { /Verse.org/SceneGraph }


# Helper function to return an array (instead of a generator) of all components attached to descendant entities to and including `InEntity` of type `component_type`.
(Entity:entity).FindDescendantComponentsArray<public>(ComponentType:castable_subtype(component))<transacts>:[]ComponentType =
    for (Comp : Entity.FindDescendantComponents(ComponentType)) do Comp

# Removes an element from the given map and returns a new map without that element
(Map:[keytype]valuetype where keytype:subtype(comparable), valuetype:type).RemoveKey(ElementToRemove:keytype)<transacts>:[keytype]valuetype=
    var NewMap:[keytype]valuetype = map{}
    for (Key -> Value : Map, Key <> ElementToRemove):
        set NewMap = ConcatenateMaps(NewMap, map{Key => Value})
    return NewMap

StringToMessage<localizes>(InString : string) : message = "{InString}"